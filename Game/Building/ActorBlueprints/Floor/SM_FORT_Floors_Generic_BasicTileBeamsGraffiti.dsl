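; Floor building actor blueprint for a generic tile-and-beam floor piece with graffiti variants.
; Its construction script runs the parent's, then creates a dynamic material instance from a graffiti material variable on slot 0 of the static mesh.
; That instance takes its flipbook set, graffiti texture, background decal texture and decal scale from blueprint variables, so the graffiti look can vary without separate materials.
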(fn ConstructionScript
  (|CallParentFunction)
  (bind _returnvalue (Rendering|Material|CreateDynamicMaterialInstance (Variables|Debug|GetStaticMeshComponent) 0 (Variables|Graffiti|GetGraffitiMaterial) "None" :!class /Script/Engine.PrimitiveComponent))
  (Rendering|Material|SetScalarParameterValue _returnvalue "Graffiti_FlipBookSet" (Variables|Graffiti|GetGraffiti_FlipBookSet) :!class /Script/Engine.MaterialInstanceDynamic)
  (Rendering|Material|SetTextureParameterValue _returnvalue "Graffiti" (Variables|Graffiti|GetGraffiti))
  (Rendering|Material|SetScalarParameterValue _returnvalue "GraffitiExtraScale" (Variables|Graffiti|GetBackgroundGraffitiScale) :!class /Script/Engine.MaterialInstanceDynamic)
  (Rendering|Material|SetTextureParameterValue _returnvalue "Decal" (Variables|Graffiti|GetBackground%20Graffiti%20%28Decal%29)))
