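; Blueprint actor for a wood mines floor piece in the building system.
; On BeginPlay it assigns the WoodStressLoop sound cue to its WoodStressSound audio component.
; The construction script is empty, and the graph only sets the sound without starting playback.
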(event EventBeginPlay
  (Audio|Components|Audio|SetSound (Variables|WoodMines_Floor|GetWoodStressSound) (Variables|Default|GetWoodStressLoop_SoundCue)))

(fn ConstructionScript)
